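# Sound cue for a Korean Reload announcer voice line (Pass 010) in the TimberStake map content.
# A single dialogue player node plays the matching DialogueWave, with the TimberStake announcer as speaker and no targets.
# The cue routes through the Fort_VO_Narrator sound class at reduced volume, with a very large max distance.
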SoundCue /TimberStakeMapContent/Sounds/VO/TimberStake/Korean/VO_Reload_Announcer_Pass_010_Dialogue_Cue
root=DialoguePlayer_0
nodes [{
  id=DialoguePlayer_0
  properties {dialogueWaveParameter {
    dialogueWave=/TimberStakeMapContent/Sounds/VO/TimberStake/Korean/Wavs/VO_Reload_Announcer_Pass_010_Dialogue
    context {speaker=/TimberStakeMapContent/Sounds/VO/TimberStake/Korean/Wavs/VO_TiberStake_Announcer, targets []}
   }}
 }]
properties {maxDistance=2097152, soundClassObject=/Game/Sounds/FortSoundClassesAndModes/Fort_VO_Narrator, volumeMultiplier=0.6667}
